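# Parent cloth-shaded material for the SlopeTramp pickaxe, extending the 2023 anime shader (Pre, Calculations, Post functions) with an animated eye.
# Two smoothstepped time sines shift the eye texture side to side on UV1, periodically blending it back to center.
# The texture's static green channel adds horizontal UV distortion that flips with look direction.
# The shifted red channel is added to the DFL mask and saturated before the Post function.
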Material /BRCosmetics/Weapons/FORT_Melee/Pickaxe_SlopeTramp/Materials/M_Pickaxe_SlopeTramp_Parent
graph {
 expressions {
  MaterialExpressionAdd_0 {
   class=MaterialExpressionAdd
   inputs [
    A {from=MaterialExpressionTextureCoordinate_3}
    B {from=MaterialExpressionLinearInterpolate_0}
   ]
   properties {constB=0}
  }
  MaterialExpressionAdd_1 {
   class=MaterialExpressionAdd
   inputs [
    A {from=MaterialExpressionMaterialFunctionCall_20, output=DFL_Mask}
    B {from=MaterialExpressionTextureSample_0, output=R}
   ]
  }
  MaterialExpressionAdd_2 {
   class=MaterialExpressionAdd
   inputs [
    A {from=MaterialExpressionAdd_0}
    B {from=MaterialExpressionMultiply_2}
   ]
   properties {constB=0}
  }
  MaterialExpressionConstant2Vector_0 {class=MaterialExpressionConstant2Vector, properties {r=0.2}}
  MaterialExpressionConstant2Vector_1 {class=MaterialExpressionConstant2Vector, properties {r=-0.2}}
  MaterialExpressionConstant_1 {class=MaterialExpressionConstant, properties {r=0.1}}
  MaterialExpressionConvert_0 {
   class=MaterialExpressionConvert
   inputs [
    None {from=MaterialExpressionTextureSample_5, output=G}
    None {from=""}
   ]
   properties {
    nodeName=Make Vector2
    convertInputs [
     {
      type=Scalar
      expressionInput {
       expression=&self:MaterialExpressionTextureSample_5
       inputName=None
       mask=1
       maskA=0
       maskB=0
       maskG=1
       maskR=0
       outputIndex=2
      }
     }
     {type=Scalar, expressionInput {expression=null, inputName=None, mask=1, maskA=0, maskB=0, maskG=1, maskR=0, outputIndex=2}}
    ]
    convertOutputs [{type=Vector2}]
   }
  }
  MaterialExpressionLinearInterpolate_0 {
   class=MaterialExpressionLinearInterpolate
   inputs [
    A {from=MaterialExpressionConstant2Vector_1}
    B {from=MaterialExpressionConstant2Vector_0}
    Alpha {from=MaterialExpressionLinearInterpolate_4}
   ]
   properties {constAlpha=1}
  }
  MaterialExpressionLinearInterpolate_2 {
   class=MaterialExpressionLinearInterpolate
   inputs [
    A {from=""}
    B {from=""}
    Alpha {from=MaterialExpressionLinearInterpolate_4}
   ]
   properties {constA=1, constAlpha=1, constB=-1}
  }
  MaterialExpressionLinearInterpolate_4 {
   class=MaterialExpressionLinearInterpolate
   inputs [
    A {from=MaterialExpressionSmoothStep_0}
    B {from=""}
    Alpha {from=MaterialExpressionSmoothStep_1}
   ]
   properties {constB=0.5}
  }
  MaterialExpressionMaterialFunctionCall_0 {
   class=MaterialExpressionMaterialFunctionCall
   inputs [Speed {from=MaterialExpressionConstant_1}]
   properties {materialFunction=&DebugTimeSine}
  }
  MaterialExpressionMaterialFunctionCall_1 {
   class=MaterialExpressionMaterialFunctionCall
   inputs [Speed {from=MaterialExpressionScalarParameter_1}]
   properties {materialFunction=&DebugTimeSine}
  }
  MaterialExpressionMaterialFunctionCall_14 {
   class=MaterialExpressionMaterialFunctionCall
   inputs [
    Shading_Gradient {from=MaterialExpressionMaterialFunctionCall_20, output=Shading_Gradient}
    Threshold_Gradient {from=MaterialExpressionMaterialFunctionCall_20, output=Shading_Gradient}
    Specular_Gradient {from=MaterialExpressionMaterialFunctionCall_20, output=Shading_Gradient}
    Sheen_Mask {from=MaterialExpressionMaterialFunctionCall_20, output=Shading_Gradient}
    Thickness {from=MaterialExpressionMaterialFunctionCall_20, output=Shading_Gradient}
    DFL_Mask {from=MaterialExpressionSaturate_1}
    FX {from=MaterialExpressionMaterialFunctionCall_20, output=Shading_Gradient}
    Metallic_Gradient {from=MaterialExpressionMaterialFunctionCall_20, output=Shading_Gradient}
    RimLight_Gradient {from=MaterialExpressionMaterialFunctionCall_20, output=Shading_Gradient}
    NormalBuffer {from=MaterialExpressionMaterialFunctionCall_20, output=Shading_Gradient}
    -------- {from=""}
    CINE_LightSources {from=MaterialExpressionMaterialFunctionCall_20, output=Shading_Gradient}
    CINE_LightColor {from=MaterialExpressionMaterialFunctionCall_20, output=Shading_Gradient}
    --------- {from=""}
    Color_Lit {from=MaterialExpressionMaterialFunctionCall_21, output=Color_Lit}
    Color_Shaded {from=MaterialExpressionMaterialFunctionCall_21, output=Color_Lit}
    SRM {from=MaterialExpressionMaterialFunctionCall_21, output=Color_Lit}
    Emissive {from=MaterialExpressionMaterialFunctionCall_21, output=Color_Lit}
    DFL_Color {from=MaterialExpressionMaterialFunctionCall_21, output=Color_Lit}
    Specular_Color {from=MaterialExpressionMaterialFunctionCall_21, output=Color_Lit}
    RimLight_Color {from=MaterialExpressionMaterialFunctionCall_21, output=Color_Lit}
   ]
   properties {materialFunction=/Game/Global/Materials/Character_Parent_2023/Anime_Parent_2023/MaterialFunctions/MFC_AnimeShader_2023_Post}
  }
  MaterialExpressionMaterialFunctionCall_20 {
   class=MaterialExpressionMaterialFunctionCall
   inputs [
    CustomLightVector {from=""}
    OverrideNormal_WS {from=""}
    STT {from=MaterialExpressionMaterialFunctionCall_21, output=STT}
    Normal_Map {from=MaterialExpressionMaterialFunctionCall_21, output=STT}
    DFL {from=MaterialExpressionMaterialFunctionCall_21, output=STT}
    FX {from=MaterialExpressionMaterialFunctionCall_21, output=STT}
   ]
   properties {materialFunction=/Game/Global/Materials/Character_Parent_2023/Anime_Parent_2023/MaterialFunctions/MFC_AnimeShader_2023_Calculations}
  }
  MaterialExpressionMaterialFunctionCall_21 {
   class=MaterialExpressionMaterialFunctionCall
   inputs [CustomUV {from=""}]
   properties {materialFunction=/Game/Global/Materials/Character_Parent_2023/Anime_Parent_2023/MaterialFunctions/MFC_AnimeShader_2023_Pre}
  }
  MaterialExpressionMultiply_2 {
   class=MaterialExpressionMultiply
   inputs [
    A {from=MaterialExpressionConvert_0}
    B {from=MaterialExpressionMultiply_3}
   ]
   properties {constB=0.1}
  }
  MaterialExpressionMultiply_3 {
   class=MaterialExpressionMultiply
   inputs [
    A {from=MaterialExpressionLinearInterpolate_2}
    B {from=MaterialExpressionScalarParameter_0}
   ]
   properties {constB=0.05}
  }
  MaterialExpressionNamedRerouteDeclaration_0 {
   class=MaterialExpressionNamedRerouteDeclaration
   inputs [None {from=MaterialExpressionTextureSample_0, output=R}]
   properties {nodeColor=<rgba 0.2891 0 1 1>}
  }
  MaterialExpressionSaturate_1 {class=MaterialExpressionSaturate, inputs [None {from=MaterialExpressionAdd_1}]}
  MaterialExpressionScalarParameter_0 {class=MaterialExpressionScalarParameter, properties {defaultValue=-0.08, parameterName=distortion}}
  MaterialExpressionScalarParameter_1 {class=MaterialExpressionScalarParameter, properties {defaultValue=0.25, parameterName=speed}}
  MaterialExpressionSmoothStep_0 {
   class=MaterialExpressionSmoothStep
   inputs [
    Min {from="", output=""}
    Max {from="", output=""}
    Value {from=MaterialExpressionMaterialFunctionCall_1, output=Result}
   ]
   properties {constMax=0.6, constMin=0.4}
  }
  MaterialExpressionSmoothStep_1 {
   class=MaterialExpressionSmoothStep
   inputs [
    Min {from="", output=""}
    Max {from="", output=""}
    Value {from=MaterialExpressionMaterialFunctionCall_0, output=Result}
   ]
   properties {constMax=0.55, constMin=0.45}
  }
  MaterialExpressionTextureCoordinate_3 {class=MaterialExpressionTextureCoordinate, properties {coordinateIndex=1}}
  MaterialExpressionTextureSample_0 {
   class=MaterialExpressionTextureSample
   inputs [
    UVs {from=MaterialExpressionAdd_2}
    Tex {from=""}
    Apply View MipBias {from=""}
   ]
   properties {
    samplerSource=SSM_Wrap_WorldGroupSettings
    samplerType=SAMPLERTYPE_Masks
    texture=/BRCosmetics/Weapons/FORT_Melee/Pickaxe_SlopeTramp/Textures/T_SlopeTramp_Pickaxe_Eye
   }
  }
  MaterialExpressionTextureSample_5 {
   class=MaterialExpressionTextureSample
   inputs [
    UVs {from=MaterialExpressionTextureCoordinate_3}
    Tex {from=""}
    Apply View MipBias {from=""}
   ]
   properties {
    samplerSource=SSM_Wrap_WorldGroupSettings
    samplerType=SAMPLERTYPE_Masks
    texture=&T_SlopeTramp_Pickaxe_Eye
   }
  }
 }
 properties {
  bAutomaticallySetUsageInEditor=false
  bUseMaterialAttributes=true
  bUsedWithClothing=true
  bUsedWithMeshDeformer=true
  bUsedWithMorphTargets=true
  bUsedWithNiagaraMeshParticles=true
  bUsedWithSkeletalMesh=true
  bUsedWithStaticLighting=true
  pixelDepthOffsetMode=PDOM_Legacy
  previewMesh=/BRCosmetics/Weapons/FORT_Melee/Pickaxe_SlopeTramp/Meshes/SlopeTramp_Axe
  shadingModel=MSM_Cloth
  usedShadingModels=Cloth
 }
}
