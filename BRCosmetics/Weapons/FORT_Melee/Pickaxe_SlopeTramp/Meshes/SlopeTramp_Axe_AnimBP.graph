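# Animation Blueprint for the SlopeTramp pickaxe weapon, playing idle animations and adding physics motion to its tail bones.
# An idle state machine, triggered into an idle break after random delays, feeds a slot and cached pose.
# Two four-bone AnimDynamics chains with different motion scales and speed limits branch from that pose into a bool blend.
# The event graph updates front-end/pawn state, simulation space, and an inverse alpha from the PhysicsOff curve.
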AnimBlueprint /BRCosmetics/Weapons/FORT_Melee/Pickaxe_SlopeTramp/Meshes/SlopeTramp_Axe_AnimBP
target_skeleton=&/SlopeTramp_Axe_Skeleton
graphs [
 AnimGraph {nodes [
   Input Pose()
   n1 = Anim Dynamics - Bone: dyn_tail_0!(
    SimulationSpace:AnimPhysSimSpaceType Enum=Component
    ComponentPose=Local To Component(LocalPose=Use cached pose 'Sim'())
    Alpha=1
    ExternalForce="0,0,0"
   ) -> (Pose) {
    angularDampingOverride=1
    angularSpringConstant=3000
    bAngularSpring=true
    bGravityOverrideInSimSpace=true
    bOverrideAngularDamping=true
    bUseGravityOverride=true
    boundBone {boneName=dyn_tail_0}
    componentAppliedLinearAccClamp=<3000 3000 3000>
    componentLinearAccScale {z=0.7}
    componentLinearVelScale {z=0.4}
    gravityOverride {z=-980}
    physicsBodyDefinitions [{
      boundBone {boneName=dyn_tail_0}
      boxExtents=<15 15 15>
      constraintSetup {angularLimitsMax=<5 5 5>, angularLimitsMin=<-5 -5 -5>, angularTarget=<1 0 0>}
      localJointOffset=<1 0.5 0>
     }]
    simSpaceSettings {maxAngularAcceleration=20, maxAngularVelocity=3, simSpaceAngularAlpha=1}
   }
   n2 = Anim Dynamics - Bone: dyn_tail_1!(SimulationSpace:AnimPhysSimSpaceType Enum, ComponentPose=@n1, Alpha=1, ExternalForce="0,0,0") -> (Pose) {
    angularDampingOverride=1
    angularSpringConstant=3000
    bAngularSpring=true
    bGravityOverrideInSimSpace=true
    bOverrideAngularDamping=true
    bUseGravityOverride=true
    simulationSpace=RootRelative
    boundBone {boneName=dyn_tail_1}
    componentAppliedLinearAccClamp=<3000 3000 3000>
    componentLinearAccScale {z=0.7}
    componentLinearVelScale {z=0.4}
    gravityOverride {z=-980}
    physicsBodyDefinitions [{
      boundBone {boneName=dyn_tail_1}
      boxExtents=<15 15 15>
      constraintSetup {angularLimitsMax=<10 5 10>, angularLimitsMin=<-10 -15 -10>, angularTarget=<1 0 0>}
      localJointOffset=<1 0.5 0>
     }]
    simSpaceSettings {maxAngularAcceleration=20, maxAngularVelocity=3, simSpaceAngularAlpha=1}
   }
   n3 = Anim Dynamics - Bone: dyn_tail_2!(SimulationSpace:AnimPhysSimSpaceType Enum, ComponentPose=@n2, Alpha=1, ExternalForce="0,0,0") -> (Pose) {
    angularDampingOverride=1
    angularSpringConstant=3000
    bAngularSpring=true
    bGravityOverrideInSimSpace=true
    bOverrideAngularDamping=true
    bUseGravityOverride=true
    simulationSpace=RootRelative
    boundBone {boneName=dyn_tail_2}
    componentAppliedLinearAccClamp=<3000 3000 3000>
    componentLinearAccScale {z=0.7}
    componentLinearVelScale {z=0.4}
    gravityOverride {z=-980}
    physicsBodyDefinitions [{
      boundBone {boneName=dyn_tail_2}
      boxExtents=<15 15 15>
      constraintSetup {angularLimitsMax=<10 15 20>, angularLimitsMin=<-10 -41.3738 -25>, angularTarget=<1 0 0>}
      localJointOffset=<1 0.5 0>
     }]
    simSpaceSettings {maxAngularAcceleration=20, maxAngularVelocity=3, simSpaceAngularAlpha=1}
   }
   n4 = Anim Dynamics - Bone: dyn_tail_3!(SimulationSpace:AnimPhysSimSpaceType Enum, ComponentPose=@n3, Alpha=1, ExternalForce="0,0,0") -> (Pose) {
    angularDampingOverride=1
    angularSpringConstant=3000
    bAngularSpring=true
    bGravityOverrideInSimSpace=true
    bOverrideAngularDamping=true
    bUseGravityOverride=true
    simulationSpace=RootRelative
    boundBone {boneName=dyn_tail_3}
    componentAppliedLinearAccClamp=<3000 3000 3000>
    componentLinearAccScale {z=0.7}
    componentLinearVelScale {z=0.4}
    gravityOverride {z=-980}
    physicsBodyDefinitions [{
      boundBone {boneName=dyn_tail_3}
      boxExtents=<15 15 15>
      constraintSetup {angularLimitsMax=<10 15 20>, angularLimitsMin=<-10 -15 -25>, angularTarget=<1 0 0>}
      localJointOffset=<1 0.5 0>
     }]
    simSpaceSettings {maxAngularAcceleration=20, maxAngularVelocity=3, simSpaceAngularAlpha=1}
   }
   n5 = Anim Dynamics - Bone: dyn_tail_0!(ComponentPose=Local To Component(LocalPose=Use cached pose 'Sim'()), Alpha=1, ExternalForce="0,0,0") -> (Pose) {
    angularDampingOverride=1
    angularSpringConstant=3000
    bAngularSpring=true
    bGravityOverrideInSimSpace=true
    bOverrideAngularDamping=true
    bUseGravityOverride=true
    simulationSpace=RootRelative
    boundBone {boneName=dyn_tail_0}
    componentAppliedLinearAccClamp=<3000 3000 3000>
    componentLinearAccScale {z=2}
    componentLinearVelScale {z=1}
    gravityOverride {z=-980}
    physicsBodyDefinitions [{
      boundBone {boneName=dyn_tail_0}
      boxExtents=<15 15 15>
      constraintSetup {angularLimitsMax=<5 5 5>, angularLimitsMin=<-5 -5 -5>, angularTarget=<1 0 0>}
      localJointOffset=<1 0.5 0>
     }]
    simSpaceSettings {maxAngularAcceleration=40, maxAngularVelocity=5.5, simSpaceAngularAlpha=1}
   }
   n6 = Anim Dynamics - Bone: dyn_tail_1!(ComponentPose=@n5, Alpha=1, ExternalForce="0,0,0") -> (Pose) {
    angularDampingOverride=1
    angularSpringConstant=3000
    bAngularSpring=true
    bGravityOverrideInSimSpace=true
    bOverrideAngularDamping=true
    bUseGravityOverride=true
    simulationSpace=RootRelative
    boundBone {boneName=dyn_tail_1}
    componentAppliedLinearAccClamp=<3000 3000 3000>
    componentLinearAccScale {z=2}
    componentLinearVelScale {z=1}
    gravityOverride {z=-980}
    physicsBodyDefinitions [{
      boundBone {boneName=dyn_tail_1}
      boxExtents=<15 15 15>
      constraintSetup {angularLimitsMax=<10 5 10>, angularLimitsMin=<-10 -15 -10>, angularTarget=<1 0 0>}
      localJointOffset=<1 0.5 0>
     }]
    simSpaceSettings {maxAngularAcceleration=40, maxAngularVelocity=5.5, simSpaceAngularAlpha=1}
   }
   n7 = Anim Dynamics - Bone: dyn_tail_2!(ComponentPose=@n6, Alpha=1, ExternalForce="0,0,0") -> (Pose) {
    angularDampingOverride=1
    angularSpringConstant=3000
    bAngularSpring=true
    bGravityOverrideInSimSpace=true
    bOverrideAngularDamping=true
    bUseGravityOverride=true
    simulationSpace=RootRelative
    boundBone {boneName=dyn_tail_2}
    componentAppliedLinearAccClamp=<3000 3000 3000>
    componentLinearAccScale {z=2}
    componentLinearVelScale {z=1}
    gravityOverride {z=-980}
    physicsBodyDefinitions [{
      boundBone {boneName=dyn_tail_2}
      boxExtents=<15 15 15>
      constraintSetup {angularLimitsMax=<10 15 20>, angularLimitsMin=<-10 -41.3738 -25>, angularTarget=<1 0 0>}
      localJointOffset=<1 0.5 0>
     }]
    simSpaceSettings {maxAngularAcceleration=40, maxAngularVelocity=5.5, simSpaceAngularAlpha=1}
   }
   n8 = Anim Dynamics - Bone: dyn_tail_3!(ComponentPose=@n7, Alpha=1, ExternalForce="0,0,0") -> (Pose) {
    angularDampingOverride=1
    angularSpringConstant=3000
    bAngularSpring=true
    bGravityOverrideInSimSpace=true
    bOverrideAngularDamping=true
    bUseGravityOverride=true
    simulationSpace=RootRelative
    boundBone {boneName=dyn_tail_3}
    componentAppliedLinearAccClamp=<3000 3000 3000>
    componentLinearAccScale {z=2}
    componentLinearVelScale {z=1}
    gravityOverride {z=-980}
    physicsBodyDefinitions [{
      boundBone {boneName=dyn_tail_3}
      boxExtents=<15 15 15>
      constraintSetup {angularLimitsMax=<10 15 20>, angularLimitsMin=<-10 -15 -25>, angularTarget=<1 0 0>}
      localJointOffset=<1 0.5 0>
     }]
    simSpaceSettings {maxAngularAcceleration=40, maxAngularVelocity=5.5, simSpaceAngularAlpha=1}
   }
   Output Pose(Result=Blend Poses by bool(BlendPose_0=Component To Local(ComponentPose=@n8), BlendPose_1=Component To Local(ComponentPose=@n4)))
   AnimGraphNode_StateMachine_0 = IdleBreak() {is=title AnimGraphNode_StateMachine}
   Save cached pose 'Sim'(Pose=Slot 'DefaultSlot'(Source=@AnimGraphNode_StateMachine_0)) {cacheName=Sim}
  ]}
 IdleBreak {
  entry=Idle
  path=AnimGraph.AnimGraphNode_StateMachine_0.IdleBreak
  ids=id
  nodes [
   AnimStateNode_0 = Idle()
   AnimStateNode_1 = IdleBreak() {is=title AnimStateNode}
   AnimStateTransitionNode_0 = AnimStateTransitionNode()
   AnimStateTransitionNode_1 = AnimStateTransitionNode() {bAutomaticRuleBasedOnSequencePlayerInState=true}
  ]
  links [
   {
    dst=IdleBreak
    rule=AnimGraph.AnimGraphNode_StateMachine_0.IdleBreak.AnimStateTransitionNode_0.Transition
    src=Idle
    via=AnimStateTransitionNode_0
   }
   {
    dst=Idle
    rule=AnimGraph.AnimGraphNode_StateMachine_0.IdleBreak.AnimStateTransitionNode_1.Transition
    src=IdleBreak
    via=AnimStateTransitionNode_1
   }
  ]
 }
 Idle {
  path=AnimGraph.AnimGraphNode_StateMachine_0.IdleBreak.AnimStateNode_0.Idle
  nodes [Output Animation Pose(Result=Sequence Player 'SlopeTramp_Idle_WEP'() {sequence=/BRCosmetics/Weapons/FORT_Melee/Pickaxe_SlopeTramp/Animation/SlopeTramp_Idle_WEP})]
 }
 IdleBreak {
  path=AnimGraph.AnimGraphNode_StateMachine_0.IdleBreak.AnimStateNode_1.IdleBreak
  nodes [Output Animation Pose(Result=Sequence Player 'SlopeTramp_IdleBreak_WEP'() {bLoopAnimation=false, sequence=/BRCosmetics/Weapons/FORT_Melee/Pickaxe_SlopeTramp/Animation/SlopeTramp_IdleBreak_WEP})]
 }
 Transition {
  path=AnimGraph.AnimGraphNode_StateMachine_0.IdleBreak.AnimStateTransitionNode_0.Transition
  nodes [Result(bCanEnterTransition=Get Play_IdleBreak() -> (Play_IdleBreak:bool))]
 }
 Transition {path=AnimGraph.AnimGraphNode_StateMachine_0.IdleBreak.AnimStateTransitionNode_1.Transition, nodes [Result()]}
 EventGraph {nodes [
   Event BlueprintUpdateAnimation()
   > n1 = Sequence!(execute) -> (then_0, then_1, then_2:Exec)
   > Set Is frontedn on pawn(Is frontedn on pawn:bool=AND Boolean(A=Get bIsInFrontEnd(), B=Get bIsOnPawn() -> (bIsOnPawn:bool))) -> (Output_Get:bool)
   > n2 = Branch(Condition=Get bIsInFrontEnd())
   > Set Simulation Space(Simulation Space:AnimPhysSimSpaceType Enum=World) -> (Output_Get:AnimPhysSimSpaceType Enum)
   Set Simulation Space(execute=@n2.else)
   Delay(execute=@n1.then_2, Duration=RandomFloatInRange(Min=5, Max=20))
   > Set Play_IdleBreak(Play_IdleBreak:bool=true) -> (Output_Get:bool)
   > Delay(Duration=0.1)
   > Set Play_IdleBreak(Play_IdleBreak=false)
   Set NegPhysicsAlpha(
    execute=@n1.then_1
    NegPhysicsAlpha:double=MapRangeClamped(Value=GetCurveValue(CurveName=PhysicsOff), InRangeA=0, InRangeB=1, OutRangeA=1, OutRangeB=0)
   ) -> (Output_Get:double)
   TryGetPawnOwner()
  ]}
]
variables [
 Is frontedn on pawn {type=boolean}
 NegPhysicsAlpha {type=number, default=1}
 Play_IdleBreak {type=boolean}
 Simulation Space {type=AnimPhysSimSpaceType, default=RootRelative}
]
